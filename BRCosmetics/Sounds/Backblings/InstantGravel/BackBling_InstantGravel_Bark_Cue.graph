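# Sound cue for the InstantGravel back bling's reactive bark, playing a short sequence of barks and a growl.
# A mixer layers four fixed-delay branches in time order: bark, growl, slightly quieter bark, final bark.
# Each branch picks at random with equal weights; the three bark branches share one set of React waves, and the growl has its own.
# Close-range back bling attenuation, sound class and reactive cosmetic property sheet govern playback.
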SoundCue /BRCosmetics/Sounds/Backblings/InstantGravel/BackBling_InstantGravel_Bark_Cue
root=Mixer_0
nodes [
 {id=Delay_0, children [Random_0], properties {delayMax=0.4, delayMin=0.4}}
 {id=Delay_1, children [Modulator_0], properties {delayMax=1.44, delayMin=1.44}}
 {id=Delay_2, children [Random_2], properties {delayMax=1.91, delayMin=1.91}}
 {id=Delay_3, children [Random_5], properties {delayMax=0.6, delayMin=0.6}}
 {id=Mixer_0, children [Delay_0,Delay_3,Delay_1,Delay_2], properties {inputVolume [1,1,1,1]}}
 {id=Modulator_0, children [Random_4], properties {volumeMax=0.8, volumeMin=0.8}}
 {
  id=Random_0
  children [WavePlayer_0,WavePlayer_1,WavePlayer_2,WavePlayer_3,WavePlayer_4,WavePlayer_5,WavePlayer_6]
  properties {weights [1,1,1,1,1,1,1]}
 }
 {
  id=Random_2
  children [WavePlayer_14, WavePlayer_15, WavePlayer_16, WavePlayer_17, WavePlayer_18, WavePlayer_19, WavePlayer_20]
  properties {weights [1,1,1,1,1,1,1]}
 }
 {
  id=Random_4
  children [WavePlayer_34, WavePlayer_35, WavePlayer_36, WavePlayer_37, WavePlayer_38, WavePlayer_39, WavePlayer_40]
  properties {weights [1,1,1,1,1,1,1]}
 }
 {
  id=Random_5
  children [WavePlayer_41, WavePlayer_42, WavePlayer_43, WavePlayer_44, WavePlayer_45, WavePlayer_46]
  properties {weights [1,1,1,1,1,1]}
 }
 {id=WavePlayer_0, properties {soundWaveAssetPtr=&/Backpack_InstantGravel_React_01}}
 {id=WavePlayer_1, properties {soundWaveAssetPtr=&/Backpack_InstantGravel_React_02}}
 {id=WavePlayer_14, properties {soundWaveAssetPtr=&Backpack_InstantGravel_React_01}}
 {id=WavePlayer_15, properties {soundWaveAssetPtr=&Backpack_InstantGravel_React_02}}
 {id=WavePlayer_16, properties {soundWaveAssetPtr=&/Backpack_InstantGravel_React_03}}
 {id=WavePlayer_17, properties {soundWaveAssetPtr=&/Backpack_InstantGravel_React_04}}
 {id=WavePlayer_18, properties {soundWaveAssetPtr=&/Backpack_InstantGravel_React_05}}
 {id=WavePlayer_19, properties {soundWaveAssetPtr=&/Backpack_InstantGravel_React_06}}
 {id=WavePlayer_2, properties {soundWaveAssetPtr=&Backpack_InstantGravel_React_03}}
 {id=WavePlayer_20, properties {soundWaveAssetPtr=&/Backpack_InstantGravel_React_07}}
 {id=WavePlayer_3, properties {soundWaveAssetPtr=&Backpack_InstantGravel_React_04}}
 {id=WavePlayer_34, properties {soundWaveAssetPtr=&Backpack_InstantGravel_React_07}}
 {id=WavePlayer_35, properties {soundWaveAssetPtr=&Backpack_InstantGravel_React_06}}
 {id=WavePlayer_36, properties {soundWaveAssetPtr=&Backpack_InstantGravel_React_05}}
 {id=WavePlayer_37, properties {soundWaveAssetPtr=&Backpack_InstantGravel_React_04}}
 {id=WavePlayer_38, properties {soundWaveAssetPtr=&Backpack_InstantGravel_React_03}}
 {id=WavePlayer_39, properties {soundWaveAssetPtr=&Backpack_InstantGravel_React_02}}
 {id=WavePlayer_4, properties {soundWaveAssetPtr=&Backpack_InstantGravel_React_05}}
 {id=WavePlayer_40, properties {soundWaveAssetPtr=&Backpack_InstantGravel_React_01}}
 {id=WavePlayer_41, properties {soundWaveAssetPtr=&/Backpack_InstantGravel_React_Growl_01}}
 {id=WavePlayer_42, properties {soundWaveAssetPtr=&/Backpack_InstantGravel_React_Growl_02}}
 {id=WavePlayer_43, properties {soundWaveAssetPtr=&/Backpack_InstantGravel_React_Growl_03}}
 {id=WavePlayer_44, properties {soundWaveAssetPtr=&/Backpack_InstantGravel_React_Growl_04}}
 {id=WavePlayer_45, properties {soundWaveAssetPtr=&/Backpack_InstantGravel_React_Growl_05}}
 {id=WavePlayer_46, properties {soundWaveAssetPtr=&/Backpack_InstantGravel_React_Growl_06}}
 {id=WavePlayer_5, properties {soundWaveAssetPtr=&Backpack_InstantGravel_React_06}}
 {id=WavePlayer_6, properties {soundWaveAssetPtr=&Backpack_InstantGravel_React_07}}
]
properties {
 attenuationSettings=/BRCosmetics/Sounds/Other/Attenuation/Fort_BackBling_Close_Att
 audioPropertiesSheet=/BRCosmetics/Sounds/Settings/PropertySheets/Reactive/APS_Cosmetic_Reactive_Backbling
 maxDistance=1111.11
 soundClassObject=/Game/Sounds/FortSoundClassesAndModes/Fort_Backblings
 volumeMultiplier=0.375
}
